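# Sound cue for the Immunity Carriage card buy-back horn, routed to the exterior vehicle sound class with a boosted volume.
# It has no root or sound nodes, so the file shows no audio source to play.
# A horn-specific attenuation asset handles distance behavior.
# Its concurrency override caps simultaneous instances and stops the oldest, taking precedence over the assigned vehicle engine concurrency set.
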SoundCue /ImmunityCarriage/Assets/Audio/SC_S39_ImmunityCarriage_CardBuyBack_Horn
root=null
nodes []
properties {
 attenuationSettings=/ImmunityCarriage/Assets/Audio/Settings/S39_ImmunityCarriage_Horn_Att
 bOverrideConcurrency=true
 maxDistance=12022.501
 soundClassObject=&Fort_Vehicles_Exterior
 volumeMultiplier=3
 concurrencyOverrides {maxCount=3, resolutionRule=StopOldest}
 concurrencySet [/VehicleAudio/Settings/Vehicle_Engine_Conc]
}
