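# Sound cue for the med mist cloud detonation in the ImmunityCarriage feature, routed through the gadgets sound class.
# A single detonation wave plays through a modulator at reduced volume.
# It references smoke grenade attenuation, alongside override values for low-pass distance filtering, occlusion and stereo spread.
# A concurrency override limits it to one playing instance at a time.
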SoundCue /ImmunityCarriage/Assets/Audio/SC_S39_ImmunityCarriage_MedCloud
root=Modulator_0
nodes [
 {id=Modulator_0, children [WavePlayer_0]}
 {id=WavePlayer_0, properties {soundWaveAssetPtr=&/MedMist_Detonation_1}}
]
properties {
 attenuationSettings=/Game/Sounds/Attenuation/Fort_Grenade_Smoke_Att
 bOverrideConcurrency=true
 maxDistance=6800
 soundClassObject=&Fort_Gadgets
 volumeMultiplier=0.4
 attenuationOverrides {
  bAttenuateWithLPF=true
  falloffDistance=10000
  lPFRadiusMax=10000
  occlusionLowPassFilterFrequency=18000
  occlusionTraceChannel=ECC_GameTraceChannel11
  occlusionVolumeAttenuation=0.8
  stereoSpread=2000
  attenuationShapeExtents {x=0}
 }
 concurrencyOverrides {maxCount=1}
}
