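# Sound cue for the Immunity Carriage self-revive, layering a riser, a siren, a machine power-up and a revive horn.
# The riser plays through close and distant van-activation attenuations mixed together, while the siren and power-up use the self-revive attenuation.
# The horn starts after a fixed delay of nearly ten seconds, lowered in volume and given its own attenuation.
# A top mixer balances the layers under an important in-game mission sound class.
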SoundCue /ImmunityCarriage/Assets/Audio/SC_S39_ImmunityCarriage_SelfRevive
root=Mixer_1
nodes [
 {
  id=Attenuation_1
  children [WavePlayer_6]
  properties {attenuationSettings=/ImmunityCarriage/Assets/Audio/Settings/S39_ImmunityCarriage_SelfRevive_Att}
 }
 {
  id=Attenuation_2
  children [WavePlayer_4]
  properties {attenuationSettings=&S39_ImmunityCarriage_SelfRevive_Att}
 }
 {
  id=Attenuation_3
  children [Modulator_0]
  properties {attenuationSettings=/ImmunityCarriage/Assets/Audio/Settings/S39_ImmunityCarriage_SelfRevive_Horn_Att}
 }
 {id=Attenuation_4, children [WavePlayer_3], properties {attenuationSettings=/Game/Athena/Sounds/SC_Machine/Van_Redux_Activate_Distant_Att}}
 {id=Attenuation_5, children [WavePlayer_3], properties {attenuationSettings=/Game/Athena/Sounds/SC_Machine/Fort_Van_Activate_Close_Att}}
 {id=Delay_0, children [WavePlayer_9], properties {delayMax=9.735, delayMin=9.735}}
 {id=Mixer_0, children [Attenuation_5, Attenuation_4], properties {inputVolume [1,1]}}
 {id=Mixer_1, children [Mixer_0, Attenuation_1, Attenuation_2, Attenuation_3], properties {inputVolume [1,0.5,1,1.2]}}
 {id=Modulator_0, children [Delay_0], properties {pitchMax=1, pitchMin=1, volumeMax=0.2, volumeMin=0.2}}
 {id=WavePlayer_3, properties {soundWaveAssetPtr=&/ImmunityCarriage_SelfRevive_Riser}}
 {id=WavePlayer_4, properties {soundWaveAssetPtr=/Game/Athena/Sounds/SC_Machine/SC_Machine_Red_Power_Up_Start}}
 {id=WavePlayer_6, properties {soundWaveAssetPtr=&/ImmunityCarriage_SelfRevive_Siren_1}}
 {id=WavePlayer_9, properties {soundWaveAssetPtr=&/ImmunityCarriage_Revive_Horn}}
]
properties {
 attenuationSettings=&S39_ImmunityCarriage_SelfRevive_Att
 maxDistance=14000
 soundClassObject=&Fort_Mission_Ingame_Important_PSM
 volumeMultiplier=1.2
 attenuationOverrides {bAttenuate=false, bEnableReverbSend=false, bSpatialize=false, falloffDistance=2000, attenuationShapeExtents {x=1500}}
}
