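# Sound cue for the Immunity Carriage's siren, played as an exterior vehicle sound.
# A single looping boost siren remix wave feeds a modulator with fixed pitch and raised volume, so there is no random variation.
# The cue adds its own volume multiplier on top of the modulator boost and uses a dedicated siren attenuation asset for spatial falloff.
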SoundCue /ImmunityCarriage/Assets/Audio/SC_S39_ImmunityCarriage_Siren
root=Modulator_1
nodes [
 {id=Modulator_1, children [WavePlayer_2], properties {pitchMax=1, pitchMin=1, volumeMax=1.5, volumeMin=1.5}}
 {id=WavePlayer_2, properties {bLooping=true, soundWaveAssetPtr=&/ImmunityCarriage_BoostSiren_ReMix_3}}
]
properties {
 attenuationSettings=/ImmunityCarriage/Assets/Audio/Settings/S39_ImmunityCarriage_Siren_Att
 maxDistance=15000
 soundClassObject=&Fort_Vehicles_Exterior
 volumeMultiplier=1.5
}
