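# Sound cue for the drivable Reboot Van's engine shutoff sound.
# A Mixer layers a BasicSUV MERC C63 shutoff wave with a Reboot Van-specific engine-off wave at equal volume, feeding a Modulator at default settings as the root.
# The cue boosts volume and uses shared vehicle movement attenuation, engine concurrency and the vehicle interior sound class.
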SoundCue /ImmunityCarriage/Assets/Audio/Veh_DrivableRebootVan_Engine_Off_Cue
root=Modulator_0
nodes [
 {id=Mixer_0, children [WavePlayer_0,WavePlayer_1], properties {inputVolume [1,1]}}
 {id=Modulator_0, children [Mixer_0]}
 {id=WavePlayer_0, properties {soundWaveAssetPtr=/BasicSUV/Audio/MERC_C63_Shutoff_Mono}}
 {id=WavePlayer_1, properties {soundWaveAssetPtr=&/S39_DrivableRebootVan_EngineOff}}
]
properties {
 attenuationSettings=/VehicleAudio/Settings/Vehicle_Movement_Att
 maxDistance=12512
 soundClassObject=&Fort_Vehicles_Interior
 volumeMultiplier=3
 concurrencySet [/VehicleAudio/Settings/Vehicle_Engine_Conc]
}
