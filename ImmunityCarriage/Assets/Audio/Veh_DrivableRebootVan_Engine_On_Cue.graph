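# Sound cue for the drivable Reboot Van's engine turning on, playing a single start-up wave.
# The wave player feeds a Modulator at default settings for slight random pitch and volume variation.
# The cue boosts volume, routes to the vehicle interior sound class, and uses shared vehicle movement attenuation and engine concurrency settings.
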SoundCue /ImmunityCarriage/Assets/Audio/Veh_DrivableRebootVan_Engine_On_Cue
root=Modulator_0
nodes [
 {id=Modulator_0, children [WavePlayer_1]}
 {id=WavePlayer_1, properties {soundWaveAssetPtr=&/S39_DrivableRebootVan_StartUp}}
]
properties {
 attenuationSettings=/VehicleAudio/Settings/Vehicle_Movement_Att
 maxDistance=12512
 soundClassObject=&Fort_Vehicles_Interior
 volumeMultiplier=4
 concurrencySet [/VehicleAudio/Settings/Vehicle_Engine_Conc]
}
