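# Sound cue for a downshift on the drivable Reboot Van, layering gear, exhaust and engine sounds.
# A mixer blends four randomized, modulated layers: gear shift mechanism, exhaust pops, rev-limit pops and throttle blips.
# Short random delays stagger the exhaust pops and then the rev-limit pops behind the mechanical shift and blip.
# It plays through a close-detail vehicle movement attenuation and the vehicle interior sound class.
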SoundCue /ImmunityCarriage/Assets/Audio/Veh_DrivableRebootVan_Shift_Down_Cue
root=Mixer_1
nodes [
 {id=Delay_1, children [Modulator_3], properties {delayMax=0.1, delayMin=0.05}}
 {id=Delay_2, children [Modulator_4], properties {delayMax=0.15, delayMin=0.1}}
 {id=Mixer_1, children [Modulator_2,Delay_1,Delay_2,Modulator_0], properties {inputVolume [1,0.8,2,2]}}
 {id=Modulator_0, children [Random_0], properties {pitchMax=0.95, pitchMin=1}}
 {id=Modulator_2, children [Random_2]}
 {id=Modulator_3, children [Random_3]}
 {id=Modulator_4, children [Random_4]}
 {id=Random_0, children [WavePlayer_2,WavePlayer_3,WavePlayer_4], properties {weights [1,1,1]}}
 {id=Random_2, children [WavePlayer_18, WavePlayer_19, WavePlayer_20], properties {weights [1,1,1]}}
 {id=Random_3, children [WavePlayer_14, WavePlayer_15, WavePlayer_16, WavePlayer_17], properties {weights [1,1,1,1]}}
 {id=Random_4, children [WavePlayer_21, WavePlayer_22, WavePlayer_23, WavePlayer_24, WavePlayer_25], properties {weights [1,1,1,1,1]}}
 {id=WavePlayer_14, properties {soundWaveAssetPtr=&/MERC_C63_ExhaustPop_01}}
 {id=WavePlayer_15, properties {soundWaveAssetPtr=&/MERC_C63_ExhaustPop_02}}
 {id=WavePlayer_16, properties {soundWaveAssetPtr=&/MERC_C63_ExhaustPop_03}}
 {id=WavePlayer_17, properties {soundWaveAssetPtr=&/MERC_C63_ExhaustPop_04}}
 {id=WavePlayer_18, properties {soundWaveAssetPtr=&/Veh_Valet_Gen_Gear_Shift_Mech_01}}
 {id=WavePlayer_19, properties {soundWaveAssetPtr=&/Veh_Valet_Gen_Gear_Shift_Mech_02}}
 {id=WavePlayer_2, properties {soundWaveAssetPtr=&/S39_DrivableRebootVan_ThrottleBlip_01}}
 {id=WavePlayer_20, properties {soundWaveAssetPtr=&/Veh_Valet_Gen_Gear_Shift_Mech_03}}
 {id=WavePlayer_21, properties {soundWaveAssetPtr=&/Veh_DrivableRebootVan_RevLimitPops_08}}
 {id=WavePlayer_22, properties {soundWaveAssetPtr=&/Veh_DrivableRebootVan_RevLimitPops_07}}
 {id=WavePlayer_23, properties {soundWaveAssetPtr=&/Veh_DrivableRebootVan_RevLimitPops_03}}
 {id=WavePlayer_24, properties {soundWaveAssetPtr=&/Veh_DrivableRebootVan_RevLimitPops_02}}
 {id=WavePlayer_25, properties {soundWaveAssetPtr=&/Veh_DrivableRebootVan_RevLimitPops_01}}
 {id=WavePlayer_3, properties {soundWaveAssetPtr=&/S39_DrivableRebootVan_ThrottleBlip_02}}
 {id=WavePlayer_4, properties {soundWaveAssetPtr=&/S39_DrivableRebootVan_ThrottleBlip_03}}
]
properties {
 attenuationSettings=/VehicleAudio/Settings/Vehicle_Movement_Close_Detail_Att
 maxDistance=12000
 soundClassObject=&Fort_Vehicles_Interior
 volumeMultiplier=1
}
