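# Sound cue for the mechanical upshift sound of the drivable Reboot Van.
# An equal-weight Random node picks one of three generic Valet gear-shift waves and feeds a Modulator at default settings for variation.
# It uses the vehicle interior sound class, close-detail movement attenuation and a boosted volume.
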SoundCue /ImmunityCarriage/Assets/Audio/Veh_DrivableRebootVan_Shift_Up_Cue
root=Modulator_1
nodes [
 {id=Modulator_1, children [Random_1]}
 {id=Random_1, children [WavePlayer_11, WavePlayer_12, WavePlayer_13], properties {weights [1,1,1]}}
 {id=WavePlayer_11, properties {soundWaveAssetPtr=&/Veh_Valet_Gen_Gear_Shift_Mech_01}}
 {id=WavePlayer_12, properties {soundWaveAssetPtr=&/Veh_Valet_Gen_Gear_Shift_Mech_02}}
 {id=WavePlayer_13, properties {soundWaveAssetPtr=&/Veh_Valet_Gen_Gear_Shift_Mech_03}}
]
properties {
 attenuationSettings=/VehicleAudio/Settings/Vehicle_Movement_Close_Detail_Att
 maxDistance=12000
 soundClassObject=&Fort_Vehicles_Interior
 volumeMultiplier=1.5
}
