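# Prototype Default Lit material for the ImmunityCarriage, usable on skeletal meshes, that builds animated emissive lighting.
# Pre-skinned local position, Fresnel and time feed named reroutes for height-split interior and ceiling light masks and their sine animations.
# An inner/outer glow blend is mixed with an off color using the maximum of two rerouted signals.
# Custom primitive data CPD_EnableLights blends steady glow toward that result before MF_FX_EmissiveAdjustments.
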Material /ImmunityCarriage/Assets/Materials/M_ImmunityCarriage_Emissive_PROTO
graph {
 properties {bAutomaticallySetUsageInEditor=false, bUsedWithMeshDeformer=true, bUsedWithSkeletalMesh=true, usedShadingModels=Default Lit}
 nodes [
  n1 = LinearInterpolate(
   A=VectorParameter() {parameterName=GlowColorInner, defaultValue=<rgba 0 1.757 10 1>}.RGB
   B=VectorParameter() {parameterName=GlowColorOuter, defaultValue=<rgba 0 0.1757 1 1>}.RGB
   Alpha=NamedRerouteUsage()
  )
  n2 = VectorParameter() {parameterName=GlowColorOff, defaultValue=<rgba 0.0099 0.0257 0.1 1>}
  n3 = Max(A=NamedRerouteUsage(), B=NamedRerouteUsage())
  MaterialFunctionCall(Emissive=LinearInterpolate(
    A=@n1
    B=LinearInterpolate(A=@n2.RGB, B=@n1, Alpha=@n3)
    Alpha=ScalarParameter() {bUseCustomPrimitiveData=true, parameterName=CPD_EnableLights, primitiveDataIndex=1}
   )) {materialFunction=&MF_FX_EmissiveAdjustments}
  n4 = ScalarParameter() {defaultValue=0.5, parameterName=EmissiveFresnelExp}
  NamedRerouteDeclaration(None=Saturate(None=Fresnel(ExponentIn=@n4) {baseReflectFraction=0})) {name=Fresnel, nodeColor=<rgba 0.1484 1 0 1>}
  NamedRerouteDeclaration(None=VertexInterpolator(VS=LocalPosition() {localOrigin=InstancePreSkinning}.XYZ).PS) {name=PreSkinnedLocalPos, nodeColor=<rgba 0.7109 1 0 1>}
  n5 = Multiply(A=ComponentMask(None=NamedRerouteUsage()) {r=true}) {constB=0.005}
  n6 = Multiply(A=Time() {period=1}, B=ScalarParameter() {defaultValue=1.2, parameterName=InnerLightsSpeed}) {constB=0.5}
  n7 = ConstantBiasScale(None=Sine(None=Add(A=@n5, B=@n6)))
  NamedRerouteDeclaration(None=Multiply(A=Saturate(None=@n7), B=NamedRerouteUsage())) {name=InteriorLightsAnimation, nodeColor=<rgba 0 0.8125 1 1>}
  n8 = ComponentMask(None=NamedRerouteUsage()) {b=true}
  n9 = ScalarParameter() {defaultValue=250, parameterName=CeilingLightsHeight}
  NamedRerouteDeclaration(None=Step(Y=Subtract(A=@n8, B=@n9)) {constX=0}) {name=InteriorLights, nodeColor=<rgba 1 0.8672 0 1>}
  n10 = NamedRerouteUsage()
  n11 = Step(Y=ComponentMask(None=@n10) {g=true}) {constX=0}
  n12 = ComponentMask(None=@n10) {b=true}
  n13 = ScalarParameter() {defaultValue=250, parameterName=CeilingLightsHeight}
  n14 = OneMinus(None=Step(Y=Subtract(A=@n12, B=@n13)) {constX=0})
  n15 = NamedRerouteUsage()
  n16 = Step(Y=Abs(None=ComponentMask(None=@n15) {r=true})) {constX=100}
  n17 = Multiply(A=@n14, B=OneMinus(None=@n16))
  NamedRerouteDeclaration(None=Multiply(A=@n11, B=@n17)) {name=LeftSideCeilingLights, nodeColor=<rgba 0 0.1562 1 1>}
  NamedRerouteDeclaration(None=Multiply(A=OneMinus(None=@n11), B=@n17)) {name=RightSideCeilingLights, nodeColor=<rgba 0.4766 1 0 1>}
  NamedRerouteDeclaration(None=Multiply(A=@n14, B=@n16)) {name=CeilingLightsPiece, nodeColor=<rgba 1 0 0 1>}
  NamedRerouteDeclaration(None=@n17) {name=CeilingLights, nodeColor=<rgba 0.2422 1 0 1>}
  n18 = Time() {period=1}
  n19 = Multiply(
   A=NamedRerouteUsage()
   B=MaterialFunctionCall(Multiplier=Constant() {r=2}) {materialFunction=&Pi}.Result
  ) {constB=0.5}
  n20 = Sine(None=Multiply(A=Add(A=@n18, B=@n19), B=ScalarParameter() {defaultValue=2, parameterName=CeilingLightSpeed}) {constB=0.5})
  NamedRerouteDeclaration(None=Multiply(A=ConstantBiasScale(None=@n20), B=NamedRerouteUsage())) {name=CeilingLightsAnimation, nodeColor=<rgba 0.4531 0 1 1>}
 ]
}
